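# Parent material for a character face-accessory glow, drawn as an unlit, alpha-composite shell pushed out along vertex normals.
# An inverted, remapped fresnel drives opacity and emissive, blending scene color toward a tint.
# Its exponent shifts when the camera aligns with the mesh's local X or Z axis, allowing separate side and top-down falloff.
# Results pass through MF_CharacterParent_2023_PostFX; a soft-outline function sits unconnected.
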Material /BRCosmetics/Characters/Player/Female/Medium/Bodies/F_MED_WordUnity_Phrase/Materials/M_WordUnity_Phrase_FaceAccessory_Glow_Parent
graph {
 properties {
  bAutomaticallySetUsageInEditor=false
  bUseMaterialAttributes=true
  bUsedWithClothing=true
  bUsedWithMeshDeformer=true
  bUsedWithMorphTargets=true
  bUsedWithNiagaraMeshParticles=true
  bUsedWithSkeletalMesh=true
  bUsedWithStaticLighting=true
  blendMode=BLEND_AlphaComposite
  pixelDepthOffsetMode=PDOM_Legacy
  shadingModel=MSM_Unlit
  usedShadingModels=Unlit
 }
 nodes [
  CameraVectorWS()
  Composite() {subgraphName=Construct Base Inputs}
  n1 = SceneColor!(UVs)
  n2 = VectorParameter() {parameterName=Color1, defaultValue=<rgba 0.956 0.013 1 1>}
  n3 = CameraVectorWS()
  n4 = PixelNormalWS()
  n5 = ConstantBiasScale(None=OneMinus(None=DotProduct(A=@n3, B=@n4)))
  n6 = CameraVectorWS()
  n7 = Transform(None=Constant3Vector() {constant=<rgba 1 0 0 0>}) {transformSourceType=TRANSFORMSOURCE_Local}
  n8 = Power(Base=Abs(None=DotProduct(A=@n6, B=@n7)), Exp=Constant() {r=8}) {constExponent=10}
  n9 = Constant() {r=1}
  n10 = Constant3Vector() {constant=<rgba 0 0 1 0>}
  n11 = Abs(None=DotProduct(A=@n6, B=Transform(None=@n10) {transformSourceType=TRANSFORMSOURCE_Local}))
  n12 = Constant() {r=8}
  n13 = LinearInterpolate(
   A=LinearInterpolate(
    A=ScalarParameter() {defaultValue=1, parameterName=Exponent}
    B=ScalarParameter() {defaultValue=1, parameterName=SideExp}
    Alpha=MaterialFunctionCall(In=@n8, Contrast=@n9) {materialFunction=&CheapContrast}.Result
   )
   B=ScalarParameter() {parameterName=TopDownExp}
   Alpha=MaterialFunctionCall(In=Power(Base=@n11, Exp=@n12) {constExponent=10}, Contrast=Constant() {r=1}) {materialFunction=&CheapContrast}.Result
  )
  n14 = MaterialFunctionCall(
   Input=OneMinus(None=Power(Base=@n5, Exp=@n13))
   Input Low=ScalarParameter() {parameterName=Param_1}
   Input High=ScalarParameter() {defaultValue=0.5}
  ) {materialFunction=&RemapValue_0-1}
  n15 = LinearInterpolate(A=@n1.RGB, B=@n2.RGB, Alpha=@n14.Result)
  n16 = ScalarParameter() {defaultValue=2, parameterName=ss}
  n17 = VertexNormalWS()
  n18 = ScalarParameter() {defaultValue=1, parameterName=Expand}
  n19 = Constant()
  MaterialExpressionReroute_4 = Reroute(None=@n19)
  n20 = PreSkinnedPosition()
  n21 = ShaderStageSwitch(PixelShader=VertexInterpolator(VS=@n20).PS, VertexShader=@n20)
  MaterialExpressionReroute_5 = Reroute(None=@n21)
  n22 = MaterialFunctionCall!(Material) {materialFunction=/Game/Packages/Fortress_SharedMaterials/Base_Material_Functions/CharacterBrightnessandScale_Balance}
  MaterialExpressionReroute_6 = Reroute(None=@n22.Distance Blend)
  n23 = Fresnel() {baseReflectFraction=0, exponent=1}
  MaterialExpressionReroute_7 = Reroute(None=@n23)
  MaterialFunctionCall(
   Material=SetMaterialAttributes!(
    MaterialAttributes
    Emissive Color=Multiply(A=@n15, B=@n14.Result)
    Opacity=Multiply(A=@n14.Result, B=@n16)
    World Position Offset=Multiply(A=@n17, B=@n18)
   ) {attributeSetTypes [#1, #2, #3]}
   M=@MaterialExpressionReroute_4.M
   Pre Skinned Local Position=@MaterialExpressionReroute_5.Pre Skinned Local Position
   Distance Blend=@MaterialExpressionReroute_6.Distance Blend
   Fresnel=@MaterialExpressionReroute_7.Fresnel
  ) {materialFunction=&MF_CharacterParent_2023_PostFX}
  MaterialFunctionCall!(
   GlowMeshStencilIndex=ScalarParameter() {defaultValue=41, parameterName=CustomDepth}
   UseCustomDepth=StaticBool()
   UseCustomStencil=StaticBool()
   OutlineThickness=ScalarParameter() {defaultValue=1, parameterName=Thickness}
   OutlineStrength=ScalarParameter() {parameterName=Strength}
   PixelDepthOffset=ScalarParameter() {parameterName=PDO}
   MaxBlurDepth=ScalarParameter() {parameterName=BlurDepth}
   ------EnableInnerGlow------
   InnerGlowMaxArea
   InnerGlowPower
   InnerGlowStrength
   ------LessUse------
   MeshPosition
   Distance Steps
   Radial Steps
  ) {materialFunction=/Game/Effects/Fort_Effects/Effects/Atlas/Materials/FXCommon_SoftOutline}
  Multiply(A=@n1.RGB, B=@n2.RGB)
  PinBase!(M=@n19, Pre Skinned Local Position=@n21, Distance Blend=@n22.Distance Blend, Fresnel=@n23) {reroutePins [
    M {expression=&self:MaterialExpressionReroute_4}
    Pre Skinned Local Position {expression=&self:MaterialExpressionReroute_5}
    Distance Blend {expression=&self:MaterialExpressionReroute_6}
    Fresnel {expression=&self:MaterialExpressionReroute_7}
   ]}
  PinBase!()
  Transform(None=@n6) {transformSourceType=TRANSFORMSOURCE_World, transformType=TRANSFORM_Local}
 ]
}
