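# Unlit parent material for a female player character's face accessory, giving it a banded three-color emissive look driven by view angle.
# A parameterized Fresnel is cut by two Step thresholds that select between Color1, Color2 and Color3 for hard-edged bands.
# The result passes as material attributes into MF_CharacterParent_2023_PostFX, alongside shared inputs for pre-skinned position, brightness-and-scale distance blend and a separate Fresnel.
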Material /BRCosmetics/Characters/Player/Female/Medium/Bodies/F_MED_WordUnity_Phrase/Materials/M_WordUnity_Phrase_FaceAccessory_Parent
graph {
 properties {
  bAutomaticallySetUsageInEditor=false
  bUseMaterialAttributes=true
  bUsedWithClothing=true
  bUsedWithMeshDeformer=true
  bUsedWithMorphTargets=true
  bUsedWithNiagaraMeshParticles=true
  bUsedWithSkeletalMesh=true
  bUsedWithStaticLighting=true
  pixelDepthOffsetMode=PDOM_Legacy
  shadingModel=MSM_Unlit
  usedShadingModels=Unlit
 }
 nodes [
  Composite() {subgraphName=Construct Base Inputs}
  n1 = Fresnel(ExponentIn=ScalarParameter() {defaultValue=1, parameterName=Exp})
  n2 = LinearInterpolate(
   A=VectorParameter() {parameterName=Color1, defaultValue=<rgba 0.956 0.013 1 1>}.RGB
   B=VectorParameter() {parameterName=Color2, defaultValue=<rgba 0.956 0.013 1 1>}.RGB
   Alpha=Step(Y=@n1, X=ScalarParameter() {defaultValue=0.9, parameterName=Threshold_1})
  )
  n3 = VectorParameter() {parameterName=Color3, defaultValue=<rgba 0.956 0.013 1 1>}
  n4 = Step(Y=@n1, X=ScalarParameter() {defaultValue=0.5, parameterName=Threshold_2})
  n5 = Constant()
  MaterialExpressionReroute_4 = Reroute(None=@n5)
  n6 = PreSkinnedPosition()
  n7 = ShaderStageSwitch(PixelShader=VertexInterpolator(VS=@n6).PS, VertexShader=@n6)
  MaterialExpressionReroute_5 = Reroute(None=@n7)
  n8 = MaterialFunctionCall!(Material) {materialFunction=/Game/Packages/Fortress_SharedMaterials/Base_Material_Functions/CharacterBrightnessandScale_Balance}
  MaterialExpressionReroute_6 = Reroute(None=@n8.Distance Blend)
  n9 = Fresnel() {baseReflectFraction=0, exponent=1}
  MaterialExpressionReroute_7 = Reroute(None=@n9)
  MaterialFunctionCall(
   Material=SetMaterialAttributes!(MaterialAttributes, Emissive Color=LinearInterpolate(A=@n2, B=@n3.RGB, Alpha=@n4), Opacity, World Position Offset) {attributeSetTypes [#1, #2, #3]}
   M=@MaterialExpressionReroute_4.M
   Pre Skinned Local Position=@MaterialExpressionReroute_5.Pre Skinned Local Position
   Distance Blend=@MaterialExpressionReroute_6.Distance Blend
   Fresnel=@MaterialExpressionReroute_7.Fresnel
  ) {materialFunction=&MF_CharacterParent_2023_PostFX}
  PinBase!(M=@n5, Pre Skinned Local Position=@n7, Distance Blend=@n8.Distance Blend, Fresnel=@n9) {reroutePins [
    M {expression=&self:MaterialExpressionReroute_4}
    Pre Skinned Local Position {expression=&self:MaterialExpressionReroute_5}
    Distance Blend {expression=&self:MaterialExpressionReroute_6}
    Fresnel {expression=&self:MaterialExpressionReroute_7}
   ]}
  PinBase!()
 ]
}
