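# Physics asset of body colliders for the F_MED_WordUnity_Phrase female medium player body, named for use as cape collision.
# Capsules cover the head, spine, pelvis, arms, hands, legs and reparented foot bones.
# The pelvis carries a second, much larger offset capsule, forming a broad volume beyond the body-fitted shapes.
# Constraints link the bodies along the skeleton hierarchy with all swing and twist motion free, so their stored angle limits are inactive.
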PhysicsAsset /BRCosmetics/Characters/Player/Female/Medium/Bodies/F_MED_WordUnity_Phrase/Meshes/F_MED_WordUnity_Phrase_CapeColliders
bodies {
 calf_l [{
   shapeType=Capsule
   length=51.239
   radius=8.0835
   shapeName=calf_l_capsule
   center=<-25.1091 -0.9585 -0.4285>
   rotation=<rot -0.126 89.0748 -90.2597>
  }]
 calf_r [{
   shapeType=Capsule
   length=51.239
   radius=8.0835
   shapeName=calf_r_capsule
   center=<25.1091 0.9585 0.4285>
   rotation=<rot -0.126 89.0748 89.7403>
  }]
 foot_reparent_l [{
   shapeType=Capsule
   length=11.8775
   radius=6.8382
   shapeName=foot_reparent_l_capsule
   center=<-2.8603 5.9149 -0.0137>
   rotation=<rot 0 6.5435 -87.3573>
  }]
 foot_reparent_r [{
   shapeType=Capsule
   length=11.8775
   radius=6.8382
   shapeName=foot_reparent_r_capsule
   center=<2.8603 -5.9149 0.0137>
   rotation=<rot 0 6.5435 92.6427>
  }]
 hand_l [{shapeType=Capsule, length=2.5302, radius=7.4859, shapeName=hand_l_capsule, center=<8.7908 0 0>, rotation=<rot 90 0 0>}]
 hand_r [{shapeType=Capsule, length=2.5302, radius=7.4859, shapeName=hand_r_capsule, center=<-8.7908 0 0>, rotation=<rot 90 180 0>}]
 head [{shapeType=Capsule, length=1, radius=20.505, shapeName=head_capsule}]
 lowerarm_l [{shapeType=Capsule, length=21.5, radius=5, shapeName=lowerarm_l_capsule, center=<13.2325 0 -0.5634>, rotation=<rot 88.7106 -0.001 -0.001>}]
 lowerarm_r [{shapeType=Capsule, length=21.5, radius=5, shapeName=lowerarm_r_capsule, center=<-13.2325 0 0.5634>, rotation=<rot 83.1841 -0.0004 179.9996>}]
 pelvis [
  {shapeType=Capsule, length=0.2, radius=19.6757, shapeName=pelvis_capsule, center=<0 -5.9566 0>}
  {shapeType=Capsule, length=55.5, radius=39.5, shapeName=pelvis_capsule, center=<-43.017 6.1719 -2.2423>, rotation=<rot 90 6.6148 0>}
 ]
 spine_01 [{shapeType=Capsule, length=1.5, radius=15.5, shapeName=spine_01_capsule, center=<0 -7.0926 0>}]
 spine_02 [{shapeType=Capsule, length=0, radius=16.5199, shapeName=spine_02_capsule, center=<0 -5.3873 0>}]
 spine_03 [{shapeType=Capsule, length=2.5, radius=16, shapeName=spine_03_capsule, center=<0 -4.086 0>}]
 spine_04 [{
   shapeType=Capsule
   length=5.4017
   radius=15.1007
   shapeName=spine_04_capsule
   center=<-1.6291 -4.7486 -0.0181>
   rotation=<rot -0.0476 -90 -0.1973>
  }]
 spine_05 [{shapeType=Capsule, length=4, radius=18.4854, shapeName=spine_05_capsule, center=<0 -1.1605 0>}]
 thigh_l [{
   shapeType=Capsule
   length=31.1639
   radius=9.0632
   shapeName=thigh_l_capsule
   center=<-20.4092 -0.0777 0.3605>
   rotation=<rot 0.2473 88.9834 -88.1872>
  }]
 thigh_r [{
   shapeType=Capsule
   length=31.1639
   radius=9.0632
   shapeName=thigh_r_capsule
   center=<20.4092 0.0777 -0.3605>
   rotation=<rot 0.2473 88.9834 91.8128>
  }]
 upperarm_l [{shapeType=Capsule, length=27.5969, radius=6.2216, shapeName=upperarm_l_capsule, center=<13.728 0 0>, rotation=<rot 90 0 0>}]
 upperarm_r [{shapeType=Capsule, length=27.5969, radius=6.2216, shapeName=upperarm_r_capsule, center=<-13.728 0 0>, rotation=<rot 90 180 0>}]
}
constraints [
 {
  bone1Name=calf_l
  bone2Name=thigh_l
  swing1LimitDegrees=45
  swing1Motion=Free
  swing2LimitDegrees=45
  swing2Motion=Free
  twistLimitDegrees=45
  twistMotion=Free
 }
 {
  bone1Name=calf_r
  bone2Name=thigh_r
  swing1LimitDegrees=45
  swing1Motion=Free
  swing2LimitDegrees=45
  swing2Motion=Free
  twistLimitDegrees=45
  twistMotion=Free
 }
 {
  bone1Name=foot_reparent_l
  bone2Name=calf_l
  swing1LimitDegrees=45
  swing1Motion=Free
  swing2LimitDegrees=45
  swing2Motion=Free
  twistLimitDegrees=45
  twistMotion=Free
 }
 {
  bone1Name=thigh_l
  bone2Name=pelvis
  swing1LimitDegrees=45
  swing1Motion=Free
  swing2LimitDegrees=45
  swing2Motion=Free
  twistLimitDegrees=45
  twistMotion=Free
 }
 {
  bone1Name=thigh_r
  bone2Name=pelvis
  swing1LimitDegrees=45
  swing1Motion=Free
  swing2LimitDegrees=45
  swing2Motion=Free
  twistLimitDegrees=45
  twistMotion=Free
 }
 {
  bone1Name=spine_01
  bone2Name=pelvis
  swing1LimitDegrees=45
  swing1Motion=Free
  swing2LimitDegrees=45
  swing2Motion=Free
  twistLimitDegrees=45
  twistMotion=Free
 }
 {
  bone1Name=spine_02
  bone2Name=spine_01
  swing1LimitDegrees=45
  swing1Motion=Free
  swing2LimitDegrees=45
  swing2Motion=Free
  twistLimitDegrees=45
  twistMotion=Free
 }
 {
  bone1Name=spine_03
  bone2Name=spine_02
  swing1LimitDegrees=45
  swing1Motion=Free
  swing2LimitDegrees=45
  swing2Motion=Free
  twistLimitDegrees=45
  twistMotion=Free
 }
 {
  bone1Name=spine_05
  bone2Name=spine_04
  swing1LimitDegrees=45
  swing1Motion=Free
  swing2LimitDegrees=45
  swing2Motion=Free
  twistLimitDegrees=45
  twistMotion=Free
 }
 {
  bone1Name=spine_04
  bone2Name=spine_03
  swing1LimitDegrees=45
  swing1Motion=Free
  swing2LimitDegrees=45
  swing2Motion=Free
  twistLimitDegrees=45
  twistMotion=Free
 }
 {
  bone1Name=head
  bone2Name=spine_05
  swing1LimitDegrees=45
  swing1Motion=Free
  swing2LimitDegrees=45
  swing2Motion=Free
  twistLimitDegrees=45
  twistMotion=Free
 }
 {
  bone1Name=upperarm_l
  bone2Name=spine_05
  swing1LimitDegrees=45
  swing1Motion=Free
  swing2LimitDegrees=45
  swing2Motion=Free
  twistLimitDegrees=45
  twistMotion=Free
 }
 {
  bone1Name=upperarm_r
  bone2Name=spine_05
  swing1LimitDegrees=45
  swing1Motion=Free
  swing2LimitDegrees=45
  swing2Motion=Free
  twistLimitDegrees=45
  twistMotion=Free
 }
 {
  bone1Name=lowerarm_l
  bone2Name=upperarm_l
  swing1LimitDegrees=45
  swing1Motion=Free
  swing2LimitDegrees=45
  swing2Motion=Free
  twistLimitDegrees=45
  twistMotion=Free
 }
 {
  bone1Name=lowerarm_r
  bone2Name=upperarm_r
  swing1LimitDegrees=45
  swing1Motion=Free
  swing2LimitDegrees=45
  swing2Motion=Free
  twistLimitDegrees=45
  twistMotion=Free
 }
 {
  bone1Name=hand_l
  bone2Name=lowerarm_l
  swing1LimitDegrees=45
  swing1Motion=Free
  swing2LimitDegrees=45
  swing2Motion=Free
  twistLimitDegrees=45
  twistMotion=Free
 }
 {
  bone1Name=hand_r
  bone2Name=lowerarm_r
  swing1LimitDegrees=45
  swing1Motion=Free
  swing2LimitDegrees=45
  swing2Motion=Free
  twistLimitDegrees=45
  twistMotion=Free
 }
 {
  bone1Name=foot_reparent_r
  bone2Name=calf_r
  swing1LimitDegrees=45
  swing1Motion=Free
  swing2LimitDegrees=45
  swing2Motion=Free
  twistLimitDegrees=45
  twistMotion=Free
 }
]
