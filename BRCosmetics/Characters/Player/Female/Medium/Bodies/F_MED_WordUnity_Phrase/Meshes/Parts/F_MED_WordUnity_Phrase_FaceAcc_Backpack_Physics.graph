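# Physics asset for the FaceAcc/Backpack part of the female medium WordUnity_Phrase cosmetic, built around a dangling dyn_hair bone chain.
# The chain's base locks angularly to neck_02, and ten capsule links follow with tight angular limits, the first four also locking one swing axis.
# Capsules on torso, neck, upper arms and legs, joined by free angular constraints, provide body shapes.
# An oversized root capsule below the origin forms a floor-like surface.
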PhysicsAsset /BRCosmetics/Characters/Player/Female/Medium/Bodies/F_MED_WordUnity_Phrase/Meshes/Parts/F_MED_WordUnity_Phrase_FaceAcc_Backpack_Physics
bodies {
 calf_l [{
   shapeType=Capsule
   length=46.505
   radius=7.505
   shapeName=calf_l_capsule
   center=<-21.8972 -1.2183 0.312>
   rotation=<rot 89.447 -177.1267 179.9998>
  }]
 calf_r [{
   shapeType=Capsule
   length=46.505
   radius=7.505
   shapeName=calf_r_capsule
   center=<21.8942 1.2182 0>
   rotation=<rot 88.7258 -177.1265 -0.0001>
  }]
 clavicle_l [{shapeType=Capsule, length=4, radius=7.005, shapeName=clavicle_l_capsule, center=<4.2282 0 0>, rotation=<rot 90 0 0>}]
 clavicle_r [{shapeType=Capsule, length=4, radius=7.005, shapeName=clavicle_r_capsule, center=<-4.2282 0 0>, rotation=<rot 90 180 0>}]
 dyn_hair_01 [{shapeType=Capsule, length=13.4575, radius=8.4579, shapeName=dyn_hair_01_capsule, center=<13.4017 0 -0.9795>, rotation=<rot 0 0 90>}]
 dyn_hair_02 [{shapeType=Capsule, length=13.2297, radius=6.226, shapeName=dyn_hair_02_capsule, center=<14.2793 0 0>, rotation=<rot 0 0 90>}]
 dyn_hair_03 [{shapeType=Capsule, length=4.2662, radius=8.7501, shapeName=dyn_hair_03_capsule, center=<8.3432 0 0>, rotation=<rot 0 90 90>}]
 dyn_hair_04 [{shapeType=Capsule, length=3.8174, radius=8.9428, shapeName=dyn_hair_04_capsule, center=<6.7336 0 0>, rotation=<rot 0 90 90>}]
 dyn_hair_05 [{shapeType=Capsule, length=4.6219, radius=7.9126, shapeName=dyn_hair_05_capsule, center=<6.8514 0 0>, rotation=<rot 0 90 90>}]
 dyn_hair_06 [{shapeType=Capsule, length=5.4677, radius=6.9649, shapeName=dyn_hair_06_capsule, center=<6.8244 0 0>, rotation=<rot 0 90 90>}]
 dyn_hair_07 [{shapeType=Capsule, length=5.7649, radius=6.0241, shapeName=dyn_hair_07_capsule, center=<6.5399 0 0>, rotation=<rot 0 90 90>}]
 dyn_hair_08 [{shapeType=Capsule, length=3.8821, radius=5.6123, shapeName=dyn_hair_08_capsule, center=<5.4568 0 0>, rotation=<rot 90 0 0>}]
 dyn_hair_09 [{shapeType=Capsule, length=3.5861, radius=4.7716, shapeName=dyn_hair_09_capsule, center=<5.4568 0 0>, rotation=<rot 90 0 0>}]
 dyn_hair_10 [{shapeType=Capsule, length=3.3876, radius=3.6244, shapeName=dyn_hair_10_capsule, center=<5.4568 0 0>, rotation=<rot 90 0 0>}]
 dyn_hair_base [{shapeType=Capsule, length=0.505, radius=0.505, shapeName=dyn_hair_base_capsule, rotation=<rot 0 -90 0>}]
 foot_l [{shapeType=Capsule, length=11, radius=6.005, shapeName=foot_l_capsule, center=<0 4.6437 0>, rotation=<rot 0 0 90>}]
 foot_r [{shapeType=Capsule, length=11, radius=6.005, shapeName=foot_r_capsule, center=<0 -4.6437 0>, rotation=<rot 0 0 -90>}]
 neck_01 [{shapeType=Capsule, length=0, radius=12.005, shapeName=neck_01_capsule, center=<0 8.0594 0>}]
 neck_02 [{shapeType=Capsule, length=0, radius=14.005, shapeName=neck_02_capsule, center=<4.8329 8.1857 0>}]
 pelvis [{shapeType=Capsule, length=6, radius=10.005, shapeName=pelvis_capsule}]
 root [{shapeType=Capsule, length=150, radius=250, shapeName=root_capsule, center=<0 0 -245>, rotation=<rot 0 0 90>}]
 spine_01 [{shapeType=Capsule, length=6, radius=8.005, shapeName=spine_01_capsule}]
 spine_02 [{shapeType=Capsule, length=6, radius=7.005, shapeName=spine_02_capsule}]
 spine_03 [{shapeType=Capsule, length=7, radius=6.505, shapeName=spine_03_capsule}]
 spine_04 [{shapeType=Capsule, length=6, radius=6.005, shapeName=spine_04_capsule}]
 spine_05 [{shapeType=Capsule, length=0, radius=13.505, shapeName=spine_05_capsule, center=<0 6.2431 0>}]
 thigh_l [{shapeType=Capsule, length=36, radius=7.005, shapeName=thigh_l_capsule, center=<-18.5768 0 0>, rotation=<rot 90 0 0>}]
 thigh_r [{shapeType=Capsule, length=36, radius=7.005, shapeName=thigh_r_capsule, center=<18.5768 0 0>, rotation=<rot 90 180 0>}]
 upperarm_l [{shapeType=Capsule, length=26.5, radius=4.505, shapeName=upperarm_l_capsule, center=<12.3351 0 0>, rotation=<rot 90 0 0>}]
 upperarm_r [{shapeType=Capsule, length=22, radius=5.005, shapeName=upperarm_r_capsule, center=<-12.3351 0 0>, rotation=<rot 90 180 0>}]
}
constraints [
 {
  bone1Name=dyn_hair_01
  bone2Name=dyn_hair_base
  swing1LimitDegrees=5
  swing1Motion=Locked
  swing2LimitDegrees=15
  twistLimitDegrees=5
  twistMotion=Locked
 }
 {bone1Name=dyn_hair_02, bone2Name=dyn_hair_01, swing1LimitDegrees=10, swing1Motion=Locked, swing2LimitDegrees=30, twistLimitDegrees=2}
 {bone1Name=dyn_hair_03, bone2Name=dyn_hair_02, swing1LimitDegrees=5, swing1Motion=Locked, swing2LimitDegrees=30, twistLimitDegrees=5}
 {bone1Name=dyn_hair_04, bone2Name=dyn_hair_03, swing1LimitDegrees=5, swing1Motion=Locked, swing2LimitDegrees=20, twistLimitDegrees=5}
 {bone1Name=dyn_hair_05, bone2Name=dyn_hair_04, swing1LimitDegrees=3, swing2LimitDegrees=12, twistLimitDegrees=5}
 {bone1Name=dyn_hair_06, bone2Name=dyn_hair_05, swing1LimitDegrees=3, swing2LimitDegrees=12, twistLimitDegrees=5}
 {bone1Name=dyn_hair_07, bone2Name=dyn_hair_06, swing1LimitDegrees=3, swing2LimitDegrees=12, twistLimitDegrees=5}
 {
  bone1Name=upperarm_l
  bone2Name=clavicle_l
  swing1LimitDegrees=45
  swing1Motion=Free
  swing2LimitDegrees=45
  swing2Motion=Free
  twistLimitDegrees=45
  twistMotion=Free
 }
 {
  bone1Name=upperarm_r
  bone2Name=clavicle_r
  swing1LimitDegrees=45
  swing1Motion=Free
  swing2LimitDegrees=45
  swing2Motion=Free
  twistLimitDegrees=45
  twistMotion=Free
 }
 {
  bone1Name=dyn_hair_base
  bone2Name=neck_02
  swing1LimitDegrees=45
  swing1Motion=Locked
  swing2LimitDegrees=45
  swing2Motion=Locked
  twistLimitDegrees=45
  twistMotion=Locked
 }
 {
  bone1Name=neck_02
  bone2Name=neck_01
  swing1LimitDegrees=45
  swing1Motion=Free
  swing2LimitDegrees=45
  swing2Motion=Free
  twistLimitDegrees=45
  twistMotion=Free
 }
 {
  bone1Name=neck_01
  bone2Name=spine_05
  swing1LimitDegrees=45
  swing1Motion=Free
  swing2LimitDegrees=45
  swing2Motion=Free
  twistLimitDegrees=45
  twistMotion=Free
 }
 {
  bone1Name=clavicle_l
  bone2Name=spine_05
  swing1LimitDegrees=45
  swing1Motion=Free
  swing2LimitDegrees=45
  swing2Motion=Free
  twistLimitDegrees=45
  twistMotion=Free
 }
 {
  bone1Name=clavicle_r
  bone2Name=spine_05
  swing1LimitDegrees=45
  swing1Motion=Free
  swing2LimitDegrees=45
  swing2Motion=Free
  twistLimitDegrees=45
  twistMotion=Free
 }
 {
  bone1Name=spine_05
  bone2Name=spine_04
  swing1LimitDegrees=45
  swing1Motion=Free
  swing2LimitDegrees=45
  swing2Motion=Free
  twistLimitDegrees=45
  twistMotion=Free
 }
 {
  bone1Name=spine_04
  bone2Name=spine_03
  swing1LimitDegrees=45
  swing1Motion=Free
  swing2LimitDegrees=45
  swing2Motion=Free
  twistLimitDegrees=45
  twistMotion=Free
 }
 {
  bone1Name=spine_03
  bone2Name=spine_02
  swing1LimitDegrees=45
  swing1Motion=Free
  swing2LimitDegrees=45
  swing2Motion=Free
  twistLimitDegrees=45
  twistMotion=Free
 }
 {
  bone1Name=spine_02
  bone2Name=spine_01
  swing1LimitDegrees=45
  swing1Motion=Free
  swing2LimitDegrees=45
  swing2Motion=Free
  twistLimitDegrees=45
  twistMotion=Free
 }
 {
  bone1Name=spine_01
  bone2Name=pelvis
  swing1LimitDegrees=45
  swing1Motion=Free
  swing2LimitDegrees=45
  swing2Motion=Free
  twistLimitDegrees=45
  twistMotion=Free
 }
 {
  bone1Name=thigh_l
  bone2Name=pelvis
  swing1LimitDegrees=45
  swing1Motion=Free
  swing2LimitDegrees=45
  swing2Motion=Free
  twistLimitDegrees=45
  twistMotion=Free
 }
 {
  bone1Name=calf_l
  bone2Name=thigh_l
  swing1LimitDegrees=45
  swing1Motion=Free
  swing2LimitDegrees=45
  swing2Motion=Free
  twistLimitDegrees=45
  twistMotion=Free
 }
 {
  bone1Name=foot_l
  bone2Name=calf_l
  swing1LimitDegrees=45
  swing1Motion=Free
  swing2LimitDegrees=45
  swing2Motion=Free
  twistLimitDegrees=45
  twistMotion=Free
 }
 {
  bone1Name=thigh_r
  bone2Name=pelvis
  swing1LimitDegrees=45
  swing1Motion=Free
  swing2LimitDegrees=45
  swing2Motion=Free
  twistLimitDegrees=45
  twistMotion=Free
 }
 {
  bone1Name=calf_r
  bone2Name=thigh_r
  swing1LimitDegrees=45
  swing1Motion=Free
  swing2LimitDegrees=45
  swing2Motion=Free
  twistLimitDegrees=45
  twistMotion=Free
 }
 {
  bone1Name=foot_r
  bone2Name=calf_r
  swing1LimitDegrees=45
  swing1Motion=Free
  swing2LimitDegrees=45
  swing2Motion=Free
  twistLimitDegrees=45
  twistMotion=Free
 }
 {
  bone1Name=pelvis
  bone2Name=root
  swing1LimitDegrees=45
  swing1Motion=Free
  swing2LimitDegrees=45
  swing2Motion=Free
  twistLimitDegrees=45
  twistMotion=Free
 }
 {bone1Name=dyn_hair_08, bone2Name=dyn_hair_07, swing1LimitDegrees=3, swing2LimitDegrees=12, twistLimitDegrees=5}
 {bone1Name=dyn_hair_09, bone2Name=dyn_hair_08, swing1LimitDegrees=3, swing2LimitDegrees=12, twistLimitDegrees=5}
 {bone1Name=dyn_hair_10, bone2Name=dyn_hair_09, swing1LimitDegrees=3, swing2LimitDegrees=12, twistLimitDegrees=5}
]
