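# Control Rig for a female character's face accessory that poses a long hair chain and two shoulder hair correction bones procedurally from body motion.
# A Parents function constrains control nulls to the head, spine and pelvis, then TwistBones twists the chain and a Hermite, non-stretch SplineIK fits it through controls gathered at construction.
# Spherical pose readers on the head and clavicle pitch and yaw deltas weight offsets and constraints on the left and right correction bones.
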ControlRigBlueprint /BRCosmetics/Characters/Player/Female/Medium/Bodies/F_MED_WordUnity_Phrase/Meshes/Parts/F_MED_WordUnity_Phrase_FaceAcc_CtrlRig
graphs [
 New Function_ContainedGraph {nodes [
   RigVMFunctionEntryNode()
   > &RigUnit_ParentConstraint(
    AdvancedSettings=(InterpolationType=Average,RotationOrderForFilter=XZY)
    Child=(Type=Null,Name="dyn_hair_01_null")
    Filter=(TranslationFilter=(bX=true,bY=true,bZ=true),RotationFilter=(bX=true,bY=true,bZ=true),ScaleFilter=(bX=True,bY=True,bZ=True))
    Parents=((Item=(Type=Bone,Name="head"),Weight=1),(Item=(Type=Bone,Name="spine_05"),Weight=0))
    Weight=1
   )
   > &RigUnit_ParentConstraint(
    AdvancedSettings=(InterpolationType=Average,RotationOrderForFilter=XZY)
    Child=(Type=Null,Name="dyn_hair_02_null")
    Parents=((Item=(Type=Bone,Name="head"),Weight=0),(Item=(Type=Bone,Name="spine_04"),Weight=1))
    Weight=1
   )
   > &RigUnit_ParentConstraint(
    AdvancedSettings=(InterpolationType=Average,RotationOrderForFilter=XZY)
    Child=(Type=Null,Name="dyn_hair_03_null")
    Parents=((Item=(Type=Bone,Name="head"),Weight=0),(Item=(Type=Bone,Name="spine_03"),Weight=1))
    Weight=1
   )
   > &RigUnit_ParentConstraint(
    AdvancedSettings=(InterpolationType=Average,RotationOrderForFilter=XZY)
    Child=(Type=Null,Name="dyn_hair_04_null")
    Filter=(TranslationFilter=(bX=true,bY=true,bZ=true),RotationFilter=(bX=true,bY=true,bZ=true),ScaleFilter=(bX=True,bY=True,bZ=True))
    Parents=((Item=(Type=Bone,Name="head"),Weight=0),(Item=(Type=Bone,Name="spine_03"),Weight=1))
    Weight=1
   )
   > &RigUnit_ParentConstraint(
    AdvancedSettings=(InterpolationType=Average,RotationOrderForFilter=XZY)
    Child=(Type=Null,Name="dyn_hair_05_null")
    Filter=(TranslationFilter=(bX=True,bY=True,bZ=True),RotationFilter=(bX=true,bY=true,bZ=true),ScaleFilter=(bX=True,bY=True,bZ=True))
    Parents=((Item=(Type=Bone,Name="head"),Weight=0),(Item=(Type=Bone,Name="spine_01"),Weight=1))
    Weight=1
   )
   > &RigUnit_ParentConstraint(
    AdvancedSettings=(InterpolationType=Average,RotationOrderForFilter=XZY)
    Child=(Type=Null,Name="dyn_hair_06_null")
    Filter=(TranslationFilter=(bX=True,bY=True,bZ=True),RotationFilter=(bX=true,bY=true,bZ=true),ScaleFilter=(bX=True,bY=True,bZ=True))
    Parents=((Item=(Type=Bone,Name="head"),Weight=0),(Item=(Type=Bone,Name="pelvis"),Weight=1))
    Weight=1
   )
   > &RigUnit_ParentConstraint(
    AdvancedSettings=(InterpolationType=Average,RotationOrderForFilter=XZY)
    Child=(Type=Null,Name="dyn_hair_07_null")
    Filter=(TranslationFilter=(bX=true,bY=true,bZ=true),RotationFilter=(bX=true,bY=true,bZ=true),ScaleFilter=(bX=True,bY=True,bZ=True))
    Parents=((Item=(Type=Bone,Name="head"),Weight=0),(Item=(Type=Bone,Name="pelvis"),Weight=1))
    Weight=1
   )
   > &RigUnit_ParentConstraint(
    AdvancedSettings=(InterpolationType=Average,RotationOrderForFilter=XZY)
    Child=(Type=Null,Name="dyn_hair_08_null")
    Filter=(TranslationFilter=(bX=true,bY=true,bZ=true),RotationFilter=(bX=true,bY=true,bZ=true),ScaleFilter=(bX=True,bY=True,bZ=True))
    Parents=((Item=(Type=Bone,Name="head"),Weight=0),(Item=(Type=Bone,Name="pelvis"),Weight=1))
    Weight=1
   )
   > &RigUnit_ParentConstraint(
    AdvancedSettings=(InterpolationType=Average,RotationOrderForFilter=XZY)
    Child=(Type=Null,Name="dyn_hair_09_null")
    Filter=(TranslationFilter=(bX=true,bY=true,bZ=true),RotationFilter=(bX=true,bY=true,bZ=true),ScaleFilter=(bX=True,bY=True,bZ=True))
    Parents=((Item=(Type=Bone,Name="head"),Weight=0),(Item=(Type=Bone,Name="pelvis"),Weight=1))
    Weight=1
   )
   > &RigUnit_ParentConstraint(
    AdvancedSettings=(InterpolationType=Average,RotationOrderForFilter=XZY)
    Child=(Type=Null,Name="dyn_hair_10_null")
    Filter=(TranslationFilter=(bX=true,bY=true,bZ=true),RotationFilter=(bX=true,bY=true,bZ=true),ScaleFilter=(bX=True,bY=True,bZ=True))
    Parents=((Item=(Type=Bone,Name="head"),Weight=0),(Item=(Type=Bone,Name="pelvis"),Weight=1))
    Weight=1
   )
   > RigVMFunctionReturnNode()
  ]}
 RigVMFunctionLibrary {nodes [Parents = RigVMCollapseNode(ExecuteContext=())]}
 RigVMModel {nodes [
   RigUnit_BeginExecution = &RigUnit_BeginExecution()
   Parents = RigVMFunctionReferenceNode(ExecuteContext=@RigUnit_BeginExecution.ExecutePin)
   ItemArray_2 = RigVMDispatchNode(Value="((Type=Bone,Name=\"dyn_hair_01\"),(Type=Bone,Name=\"dyn_hair_02\"),(Type=Bone,Name=\"dyn_hair_03\"),(Type=Bone,Name=\"dyn_hair_04\"),(Type=Bone,Name=\"dyn_hair_05\"),(Type=Bone,Name=\"dyn_hair_06\"),(Type=Bone,Name=\"dyn_hair_07\"),(Type=Bone,Name=\"dyn_hair_08\"),(Type=B ...(truncated)") {template="DISPATCH_RigVMDispatch_Constant(io Value)"}
   > TwistBones_1 = RigVMFunctionReferenceNode(
    Axis to Twist=(bX=True,bY=true,bZ=true)
    Debug=false
    End=(Type=Bone,Name="neck_01")
    Start=(Type=Bone,Name="pelvis")
    Bones to Twist=@ItemArray_2.Value
   )
   Make_Array_1 = RigVMDispatchNode(Values="(((),Type=Bone,Name=\"dyn_hair_02\"),((),Type=Bone,Name=\"dyn_hair_03\"),((),Type=Bone,Name=\"dyn_hair_04\"),((),Type=Bone,Name=\"dyn_hair_05\"),((),Type=Bone,Name=\"dyn_hair_06\"),((),Type=Bone,Name=\"dyn_hair_07\"),((),Type=Bone,Name=\"dyn_hair_08\"),((),Type=Bone,Nam ...(truncated)") {template="DISPATCH_RigVMDispatch_ArrayMake(in Values,out Array)"}
   > SplineIK = RigVMFunctionReferenceNode(
    Primary Axis=(X=0.1,Y=0,Z=0)
    Samples Per Segment=3
    Secondary Spline Direction=(X=0,Y=0,Z=0)
    Spline Mode=Hermite
    Stretch=false
    Up Axis=(X=0,Y=0,Z=0)
    debug=false
    Bones=@Make_Array_1.Array
    Controls=RigVMVariableNode() {variable=dyn_Hair_ctrls}.Value
   )
   > Get Delta Ref = RigVMFunctionReferenceNode(Item=(Type=Bone,Name="clavicle_r"))
   > n1 = &RigUnit_SphericalPoseReader(
    ActiveRegionScaleFactors=(PositiveWidth=1,NegativeWidth=1,PositiveHeight=1,NegativeHeight=1)
    ActiveRegionSize=0.3
    Debug=(bDrawDebug=False,bDraw2D=False,bDrawLocalAxes=False,DebugScale=25,DebugSegments=20,DebugThickness=0.25)
    DriverAxis=(X=0,Y=1,Z=0)
    DriverItem=(Type=Bone,Name="head")
    FalloffRegionScaleFactors=(PositiveWidth=1,NegativeWidth=1,PositiveHeight=1,NegativeHeight=1)
    FalloffSize=0.5
    FlipHeightScaling=False
    FlipWidthScaling=False
    OptionalParentItem=(Type=Bone,Name="spine_05")
    RotationOffset=(X=0,Y=-90,Z=0)
   )
   n2 = &RigVMFunction_MathDoubleRemap(SourceMaximum=25, SourceMinimum=0, TargetMaximum=1, TargetMinimum=0, Value=@"Get Delta Ref".Result.Rotation.Pitch)
   Multiply = &RigVMFunction_MathDoubleMul(B=@n2.Result, A=@n1.OutputParam)
   > &RigUnit_ModifyTransforms(
    ItemToModify=((Item=(Type=Bone,Name="correction_hair_r"),Transform=(Rotation=(X=0,Y=0,Z=0,W=1),Translation=(X=0,Y=1,Z=-1),Scale3D=(X=1,Y=1,Z=1))))
    WeightMinimum=0
    Weight=@Multiply.Result
   )
   > &RigUnit_ModifyTransforms(
    ItemToModify=((Item=(Type=Bone,Name="correction_hair_r"),Transform=(Rotation=(X=0,Y=0,Z=0,W=1),Translation=(X=0,Y=1.5,Z=-2.5),Scale3D=(X=1,Y=1,Z=1))))
    WeightMinimum=0
    Weight=@n2.Result
   )
   Remap = &RigVMFunction_MathDoubleRemap(SourceMaximum=-1, SourceMinimum=0, TargetMaximum=1, TargetMinimum=0, Value=@"Get Delta Ref".Result.Rotation.Yaw)
   > &RigUnit_ParentConstraint(
    AdvancedSettings=(InterpolationType=Average,RotationOrderForFilter=XZY)
    Child=(Type=Bone,Name="correction_hair_r")
    Filter=(TranslationFilter=(bX=false,bY=false,bZ=false),RotationFilter=(bX=false,bY=false,bZ=true),ScaleFilter=(bX=True,bY=True,bZ=True))
    Parents=((Item=(Type=Bone,Name="clavicle_r"),Weight=1))
    Weight=@Remap.Result
   )
   > Get Delta Ref_1 = RigVMFunctionReferenceNode(Item=(Type=Bone,Name="clavicle_l"))
   > n3 = &RigUnit_SphericalPoseReader(
    ActiveRegionScaleFactors=(PositiveWidth=1,NegativeWidth=1,PositiveHeight=1,NegativeHeight=1)
    ActiveRegionSize=0.3
    Debug=(bDrawDebug=False,bDraw2D=False,bDrawLocalAxes=False,DebugScale=25,DebugSegments=20,DebugThickness=0.25)
    DriverAxis=(X=0,Y=1,Z=0)
    DriverItem=(Type=Bone,Name="head")
    FalloffRegionScaleFactors=(PositiveWidth=1,NegativeWidth=1,PositiveHeight=1,NegativeHeight=1)
    FalloffSize=0.5
    FlipHeightScaling=False
    FlipWidthScaling=False
    OptionalParentItem=(Type=Bone,Name="spine_05")
    RotationOffset=(X=0,Y=-90,Z=0)
   )
   n4 = &RigVMFunction_MathDoubleRemap(SourceMaximum=25, SourceMinimum=0, TargetMaximum=1, TargetMinimum=0, Value=@"Get Delta Ref_1".Result.Rotation.Pitch)
   > &RigUnit_ModifyTransforms(
    ItemToModify=((Item=(Type=Bone,Name="correction_hair_l"),Transform=(Rotation=(X=0,Y=0,Z=0,W=1),Translation=(X=0,Y=-1,Z=1),Scale3D=(X=1,Y=1,Z=1))))
    WeightMinimum=0
    Weight=&RigVMFunction_MathDoubleMul(B=@n4.Result, A=@n3.OutputParam).Result
   )
   > &RigUnit_ModifyTransforms(
    ItemToModify=((Item=(Type=Bone,Name="correction_hair_l"),Transform=(Rotation=(X=0,Y=0,Z=0,W=1),Translation=(X=0,Y=-1.5,Z=2.5),Scale3D=(X=1,Y=1,Z=1))))
    WeightMinimum=0
    Weight=@n4.Result
   )
   > &RigUnit_ParentConstraint(
    AdvancedSettings=(InterpolationType=Average,RotationOrderForFilter=XZY)
    Child=(Type=Bone,Name="correction_hair_l")
    Filter=(TranslationFilter=(bX=false,bY=false,bZ=false),RotationFilter=(bX=false,bY=false,bZ=True),ScaleFilter=(bX=True,bY=True,bZ=True))
    Parents=((Item=(Type=Bone,Name="clavicle_l"),Weight=1))
    Weight=&RigVMFunction_MathDoubleRemap(SourceMaximum=-1, SourceMinimum=0, TargetMaximum=1, TargetMinimum=0, Value=@"Get Delta Ref_1".Result.Rotation.Yaw).Result
   )
   > n5 = &RigUnit_SphericalPoseReader(
    ActiveRegionScaleFactors=(PositiveWidth=1,NegativeWidth=1,PositiveHeight=1,NegativeHeight=1)
    ActiveRegionSize=0.25
    Debug=(bDrawDebug=True,bDraw2D=False,bDrawLocalAxes=False,DebugScale=25,DebugSegments=20,DebugThickness=0.25)
    DriverAxis=(X=0,Y=1,Z=0)
    DriverItem=(Type=Bone,Name="head")
    FalloffRegionScaleFactors=(PositiveWidth=1,NegativeWidth=1,PositiveHeight=1,NegativeHeight=1)
    FalloffSize=0.5
    FlipHeightScaling=False
    FlipWidthScaling=False
    OptionalParentItem=(Type=Bone,Name="spine_05")
    RotationOffset=(X=180,Y=0,Z=0)
   )
   > &RigUnit_ModifyTransforms(
    ItemToModify=((Item=(Type=Bone,Name="correction_hair_l"),Transform=(Rotation=(X=0,Y=0,Z=0,W=1),Translation=(X=-10,Y=-2,Z=0),Scale3D=(X=1,Y=1,Z=1))))
    WeightMinimum=0
    Weight=@n5.OutputParam
   )
   > &RigUnit_ModifyTransforms(
    ItemToModify=((Item=(Type=Bone,Name="correction_hair_r"),Transform=(Rotation=(X=0,Y=0,Z=0,W=1),Translation=(X=0,Y=2,Z=0),Scale3D=(X=1,Y=1,Z=1))))
    WeightMinimum=0
    Weight=@n5.OutputParam
   )
   > n6 = &RigUnit_SphericalPoseReader(
    ActiveRegionScaleFactors=(PositiveWidth=1,NegativeWidth=1,PositiveHeight=1,NegativeHeight=1)
    ActiveRegionSize=0.25
    Debug=(bDrawDebug=True,bDraw2D=False,bDrawLocalAxes=False,DebugScale=25,DebugSegments=20,DebugThickness=0.25)
    DriverAxis=(X=0,Y=1,Z=0)
    DriverItem=(Type=Bone,Name="head")
    FalloffRegionScaleFactors=(PositiveWidth=1,NegativeWidth=1,PositiveHeight=1,NegativeHeight=1)
    FalloffSize=0.5
    FlipHeightScaling=False
    FlipWidthScaling=False
    OptionalParentItem=(Type=Bone,Name="spine_05")
    RotationOffset=(X=0,Y=0,Z=0)
   )
   > &RigUnit_ModifyTransforms(
    ItemToModify=((Item=(Type=Bone,Name="correction_hair_r"),Transform=(Rotation=(X=0,Y=0,Z=0,W=1),Translation=(X=10,Y=2,Z=0),Scale3D=(X=1,Y=1,Z=1))))
    WeightMinimum=0
    Weight=@n6.OutputParam
   )
   > &RigUnit_ModifyTransforms(
    ItemToModify=((Item=(Type=Bone,Name="correction_hair_l"),Transform=(Rotation=(X=0,Y=0,Z=0,W=1),Translation=(X=0,Y=-2,Z=0),Scale3D=(X=1,Y=1,Z=1))))
    WeightMinimum=0
    Weight=@n6.OutputParam
   )
  ]}
 RigVMModel Construction Event Graph {nodes [
   n1 = &RigUnit_PrepareForExecution()
   Make_Array_1 = RigVMDispatchNode(Values="(((),Type=Control,Name=\"dyn_hair_02_ctrl\"),((),Type=Control,Name=\"dyn_hair_03_ctrl\"),((),Type=Control,Name=\"dyn_hair_04_ctrl\"),((),Type=Control,Name=\"dyn_hair_05_ctrl\"),((),Type=Control,Name=\"dyn_hair_06_ctrl\"),((),Type=Control,Name=\"dyn_hair_07_ctrl\"),(() ...(truncated)") {template="DISPATCH_RigVMDispatch_ArrayMake(in Values,out Array)"}
   RigVMVariableNode(Value=@Make_Array_1.Array, ExecuteContext=@n1.ExecutePin) {variable=dyn_Hair_ctrls}
  ]}
]
hierarchy [
 root {loc=<0 0 0>, rot=<0 0 0 1>}
 pelvis {parent=root, loc=<0 -0.5301 98.0571>, rot=<0.0127 0.707 -0.0127 -0.707>}
 spine_01 {parent=pelvis, loc=<6.1208 -0.9328 0>, rot=<0 0 0.034 -0.9994>}
 spine_02 {parent=spine_01, loc=<6.5997 -0.2095 0>, rot=<0 0 0.0238 -0.9997>}
 spine_03 {parent=spine_02, loc=<6.4914 -0.1237 0>, rot=<0 0 0.0149 -0.9999>}
 spine_04 {parent=spine_03, loc=<6.9598 0.0794 0>, rot=<0 0 -0.0132 -0.9999>}
 spine_05 {parent=spine_04, loc=<6.9987 0 0>, rot=<0 0 -0.0015 -1>}
 neck_01 {parent=spine_05, loc=<11.3138 0.1529 0>, rot=<0 0 -0.2003 -0.9797>}
 neck_02 {parent=neck_01, loc=<4.5627 0 0>, rot=<0 0 0.0504 -0.9987>}
 head {parent=neck_02, loc=<5.3914 0 0>, rot=<0 0 0.115 -0.9934>}
 dyn_hair_base {parent=head, loc=<5.9966 -7.8761 0>, rot=<-0.1745 0.6852 0.6852 -0.1745>}
 dyn_hair_01 {parent=dyn_hair_base, loc=<0 0 0>, rot=<0 0 0 1>}
 dyn_hair_02 {parent=dyn_hair_01, loc=<14 0 0>, rot=<0 0.1733 0 -0.9849>}
 dyn_hair_03 {parent=dyn_hair_02, loc=<16 0 0>, rot=<0 0.0416 0 -0.9991>}
 dyn_hair_04 {parent=dyn_hair_03, loc=<12.957 0 0>, rot=<0 0.0294 0 -0.9996>}
 dyn_hair_05 {parent=dyn_hair_04, loc=<11 0 0>, rot=<0 0 0 1>}
 dyn_hair_06 {parent=dyn_hair_05, loc=<11 0 0>, rot=<0 0.0147 0 -0.9999>}
 dyn_hair_07 {parent=dyn_hair_06, loc=<11 0 0>, rot=<0 0 0 1>}
 dyn_hair_08 {parent=dyn_hair_07, loc=<11 0 0>, rot=<0 0 0 1>}
 dyn_hair_09 {parent=dyn_hair_08, loc=<10 0 0>, rot=<0 0 0 1>}
 dyn_hair_10 {parent=dyn_hair_09, loc=<9 0 0>, rot=<0 0 0 1>}
 clavicle_l {parent=spine_05, loc=<5.4831 -0.1053 -3.2946>, rot=<0.0465 -0.7289 -0.009 -0.6829>}
 upperarm_l {parent=clavicle_l, loc=<9.8658 -1.2136 -0.4105>, rot=<0.0164 -0.3245 -0.0269 -0.9454>}
 lowerarm_l {parent=upperarm_l, loc=<26.465 -0.8345 0.259>, rot=<0.0036 0.0161 -0.2298 -0.9731>}
 hand_l {parent=lowerarm_l, loc=<22.7355 -0.3203 -0.6105>, rot=<-0.6784 -0.0223 0.0735 -0.7307>}
 clavicle_r {parent=spine_05, loc=<5.4831 -0.1053 3.2946>, rot=<0.7289 0.0465 -0.6829 0.009>}
 upperarm_r {parent=clavicle_r, loc=<-9.8658 1.2136 0.4105>, rot=<0.0164 -0.3245 -0.0269 -0.9454>}
 lowerarm_r {parent=upperarm_r, loc=<-26.465 0.8345 -0.259>, rot=<0.0036 0.0161 -0.2298 -0.9731>}
 hand_r {parent=lowerarm_r, loc=<-22.7355 0.3203 0.6106>, rot=<-0.6784 -0.0223 0.0735 -0.7307>}
 correction_hair_r {parent=spine_05, loc=<5.483 -0.105 3.295>, rot=<0.7289 0.0465 -0.6829 0.009>}
 correction_hair_l {parent=spine_05, loc=<5.483 -0.105 -3.295>, rot=<0.0465 -0.7289 -0.009 -0.6829>}
 thigh_l {parent=pelvis, loc=<-1.0855 -0.2109 -8.7053>, rot=<0.0009 0.0355 0.0155 -0.9993>}
 calf_l {parent=thigh_l, loc=<-42.5284 -0.0019 -0.7297>, rot=<0.0006 -0.0161 -0.0391 -0.9991>}
 foot_l {parent=calf_l, loc=<-46.7401 -2.5014 -0.592>, rot=<0.0427 -0.0238 0.0353 -0.9982>}
 thigh_r {parent=pelvis, loc=<-1.0855 -0.2109 8.7053>, rot=<0.0355 -0.0009 0.9993 0.0155>}
 calf_r {parent=thigh_r, loc=<42.5284 0.0019 0.7297>, rot=<0.0006 -0.0161 -0.0391 -0.9991>}
 foot_r {parent=calf_r, loc=<46.7401 2.5014 0.592>, rot=<0.0427 -0.0238 0.0353 -0.9982>}
 dyn_hair_01_null {parent=root, type=Null, loc=<0 -9.6127 157.9655>, rot=<-0.6064 0.6064 0.3637 0.3637>}
 dyn_hair_01_ctrl {controlType=EULER_TRANSFORM, parent=dyn_hair_01_null, type=Control, loc=<0 0 0>, rot=<0 0 0 1>}
 dyn_hair_02_null {parent=root, type=Null, loc=<0 -16.2055 145.615>, rot=<0.5342 -0.5342 -0.4633 -0.4633>}
 dyn_hair_02_ctrl {controlType=EULER_TRANSFORM, parent=dyn_hair_02_null, type=Control, loc=<0 0 0>, rot=<0 0 0 1>}
 dyn_hair_03_null {parent=root, type=Null, loc=<0 -18.4682 129.7758>, rot=<-0.5145 0.5145 0.4851 0.4851>}
 dyn_hair_03_ctrl {controlType=EULER_TRANSFORM, parent=dyn_hair_03_null, type=Control, loc=<0 0 0>, rot=<0 0 0 1>}
 clavicle_l_ctrl {controlType=EULER_TRANSFORM, type=Control, loc=<0 0 0>, rot=<0 0 0 1>}
 clavicle_r_ctrl {controlType=EULER_TRANSFORM, type=Control, loc=<0 0 0>, rot=<0 0 0 1>}
 dyn_hair_04_null {type=Null, loc=<0 -19.2291 116.8412>, rot=<0.5 -0.5 -0.5 -0.5>}
 dyn_hair_04_ctrl {controlType=EULER_TRANSFORM, parent=dyn_hair_04_null, type=Control, loc=<0 0 0>, rot=<0 0 0 1>}
 dyn_hair_05_null {type=Null, loc=<0 -19.2291 105.8412>, rot=<-0.5 0.5 0.5 0.5>}
 dyn_hair_05_ctrl {controlType=EULER_TRANSFORM, parent=dyn_hair_05_null, type=Control, loc=<0 0 0>, rot=<0 0 0 1>}
 dyn_hair_06_null {type=Null, loc=<0 -19.2291 94.8412>, rot=<0.4926 -0.4926 -0.5073 -0.5073>}
 dyn_hair_06_ctrl {controlType=EULER_TRANSFORM, parent=dyn_hair_06_null, type=Control, loc=<0 0 0>, rot=<0 0 0 1>}
 dyn_hair_07_null {type=Null, loc=<0 -18.9059 83.8459>, rot=<-0.4926 0.4926 0.5073 0.5073>}
 dyn_hair_07_ctrl {controlType=EULER_TRANSFORM, parent=dyn_hair_07_null, type=Control, loc=<0 0 0>, rot=<0 0 0 1>}
 dyn_hair_08_null {type=Null, loc=<0 -18.5826 72.8507>, rot=<0.4926 -0.4926 -0.5073 -0.5073>}
 dyn_hair_08_ctrl {controlType=EULER_TRANSFORM, parent=dyn_hair_08_null, type=Control, loc=<0 0 0>, rot=<0 0 0 1>}
 dyn_hair_09_null {type=Null, loc=<0 -18.2887 62.855>, rot=<-0.4926 0.4926 0.5073 0.5073>}
 dyn_hair_09_ctrl {controlType=EULER_TRANSFORM, parent=dyn_hair_09_null, type=Control, loc=<0 0 0>, rot=<0 0 0 1>}
 dyn_hair_10_null {type=Null, loc=<0 -18.0243 53.8589>, rot=<0.4926 -0.4926 -0.5073 -0.5073>}
 dyn_hair_10_ctrl {controlType=EULER_TRANSFORM, parent=dyn_hair_10_null, type=Control, loc=<0 0 0>, rot=<0 0 0 1>}
]
variables [dyn_Hair_ctrls {type="TArray<FRigElementKey>"}]
