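# Control Rig for the face accessory of the F_MED_WordUnity_Phrase cosmetic body, procedurally waving its dyn_hair bone chain under the head.
# BeginExecution drives a ChainHarmonicsPerItem unit rooted at dyn_hair_01, with only Wave enabled (Pendulum and Reach are off).
# The wave oscillates on Z, eased by a sine and scaled by a curve rising along the chain.
# A second harmonics node on the same root, with different wave tuning, is not connected to execution.
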ControlRigBlueprint /BRCosmetics/Characters/Player/Female/Medium/Bodies/F_MED_WordUnity_Phrase/Meshes/Parts/F_MED_WordUnity_Phrase_FaceAcc_Harmonics_CtrlRig
graphs [
 RigVMFunctionLibrary {links [], nodes []}
 RigVMModel {nodes [
   RigUnit_BeginExecution = &RigUnit_BeginExecution()
   /Script/ControlRig.RigUnit_ChainHarmonicsPerItem(
    ChainRoot=(Type=Bone,Name="dyn_hair_01")
    DrawWorldOffset=(Rotation=(X=0,Y=0,Z=0,W=1),Translation=(X=0,Y=0,Z=0),Scale3D=(X=1,Y=1,Z=1))
    Pendulum=(bEnabled=False,PendulumStiffness=2,PendulumGravity=(X=0,Y=0,Z=0),PendulumBlend=0.75,PendulumDrag=0.98,PendulumMinimum=0,PendulumMaximum=0.1,PendulumEase=Linear,UnwindAxis=(X=0,Y=1,Z=0),UnwindMinimum=0.2,UnwindMaximum=0.05)
    Reach=(bEnabled=false,ReachTarget=(X=0,Y=0,Z=0),ReachAxis=(X=0,Y=0,Z=0),ReachMinimum=0,ReachMaximum=0,ReachEase=SineEaseInOut)
    Speed=(X=0,Y=0,Z=2)
    Wave=(bEnabled=true,WaveFrequency=(X=0,Y=0,Z=160),WaveAmplitude=(X=0,Y=0,Z=0.9),WaveOffset=(X=0,Y=0,Z=0.2),WaveNoise=(X=0,Y=0,Z=0),WaveMinimum=1.8,WaveMaximum=0.9,WaveEase=SineEaseInOut)
    WaveCurve=(EditorCurveData=(Keys=((Value=0.4),(Time=1,Value=1))))
    bDrawDebug=false
    ExecutePin=@RigUnit_BeginExecution.ExecutePin
   ) {is=struct RigVMUnitNode}
   &RigUnit_ChainHarmonicsPerItem(
    ChainRoot=(Type=Bone,Name="dyn_hair_01")
    DrawWorldOffset=(Rotation=(X=0,Y=0,Z=0,W=1),Translation=(X=0,Y=0,Z=0),Scale3D=(X=1,Y=1,Z=1))
    ExecutePin=()
    Pendulum=(bEnabled=False,PendulumStiffness=2,PendulumGravity=(X=0,Y=0,Z=0),PendulumBlend=0.75,PendulumDrag=0.98,PendulumMinimum=0,PendulumMaximum=0.1,PendulumEase=Linear,UnwindAxis=(X=0,Y=1,Z=0),UnwindMinimum=0.2,UnwindMaximum=0.05)
    Reach=(bEnabled=False,ReachTarget=(X=0,Y=0,Z=0),ReachAxis=(X=0,Y=0,Z=0),ReachMinimum=0,ReachMaximum=0,ReachEase=SineEaseInOut)
    Speed=(X=0,Y=0,Z=2.5)
    Wave=(bEnabled=true,WaveFrequency=(X=0,Y=0,Z=20),WaveAmplitude=(X=0,Y=0,Z=0.7),WaveOffset=(X=0,Y=0,Z=0.2),WaveNoise=(X=0,Y=0,Z=0),WaveMinimum=2,WaveMaximum=0.5,WaveEase=SineEaseInOut)
    WaveCurve=(EditorCurveData=(Keys=((Value=0.4),(Time=1,Value=1))))
    bDrawDebug=false
   )
  ]}
]
hierarchy [
 root {loc=<0 0 0>, rot=<0 0 0 1>}
 pelvis {parent=root, loc=<0 -0.5301 98.0571>, rot=<0.0127 0.707 -0.0127 -0.707>}
 spine_01 {parent=pelvis, loc=<6.1208 -0.9328 0>, rot=<0 0 0.034 -0.9994>}
 spine_02 {parent=spine_01, loc=<6.5997 -0.2095 0>, rot=<0 0 0.0238 -0.9997>}
 spine_03 {parent=spine_02, loc=<6.4914 -0.1237 0>, rot=<0 0 0.0149 -0.9999>}
 spine_04 {parent=spine_03, loc=<6.9598 0.0794 0>, rot=<0 0 -0.0132 -0.9999>}
 spine_05 {parent=spine_04, loc=<6.9987 0 0>, rot=<0 0 -0.0015 -1>}
 neck_01 {parent=spine_05, loc=<11.3138 0.1529 0>, rot=<0 0 -0.2003 -0.9797>}
 neck_02 {parent=neck_01, loc=<4.5627 0 0>, rot=<0 0 0.0504 -0.9987>}
 head {parent=neck_02, loc=<5.3914 0 0>, rot=<0 0 0.115 -0.9934>}
 dyn_hair_base {parent=head, loc=<5.9966 -7.8761 0>, rot=<-0.1745 0.6852 0.6852 -0.1745>}
 dyn_hair_01 {parent=dyn_hair_base, loc=<0 0 0>, rot=<0 0 0 1>}
 dyn_hair_02 {parent=dyn_hair_01, loc=<14 0 0>, rot=<0 0.1733 0 -0.9849>}
 dyn_hair_03 {parent=dyn_hair_02, loc=<16 0 0>, rot=<0 0.0416 0 -0.9991>}
 dyn_hair_04 {parent=dyn_hair_03, loc=<12.957 0 0>, rot=<0 0.0294 0 -0.9996>}
 dyn_hair_05 {parent=dyn_hair_04, loc=<11 0 0>, rot=<0 0 0 1>}
 dyn_hair_06 {parent=dyn_hair_05, loc=<11 0 0>, rot=<0 0.0147 0 -0.9999>}
 dyn_hair_07 {parent=dyn_hair_06, loc=<11 0 0>, rot=<0 0 0 1>}
 dyn_hair_08 {parent=dyn_hair_07, loc=<11 0 0>, rot=<0 0 0 1>}
 dyn_hair_09 {parent=dyn_hair_08, loc=<10 0 0>, rot=<0 0 0 1>}
 dyn_hair_10 {parent=dyn_hair_09, loc=<9 0 0>, rot=<0 0 0 1>}
 clavicle_l {parent=spine_05, loc=<5.4831 -0.1053 -3.2946>, rot=<0.0465 -0.7289 -0.009 -0.6829>}
 upperarm_l {parent=clavicle_l, loc=<9.8658 -1.2136 -0.4105>, rot=<0.0164 -0.3245 -0.0269 -0.9454>}
 lowerarm_l {parent=upperarm_l, loc=<26.465 -0.8345 0.259>, rot=<0.0036 0.0161 -0.2298 -0.9731>}
 hand_l {parent=lowerarm_l, loc=<22.7355 -0.3203 -0.6105>, rot=<-0.6784 -0.0223 0.0735 -0.7307>}
 clavicle_r {parent=spine_05, loc=<5.4831 -0.1053 3.2946>, rot=<0.7289 0.0465 -0.6829 0.009>}
 upperarm_r {parent=clavicle_r, loc=<-9.8658 1.2136 0.4105>, rot=<0.0164 -0.3245 -0.0269 -0.9454>}
 lowerarm_r {parent=upperarm_r, loc=<-26.465 0.8345 -0.259>, rot=<0.0036 0.0161 -0.2298 -0.9731>}
 hand_r {parent=lowerarm_r, loc=<-22.7355 0.3203 0.6106>, rot=<-0.6784 -0.0223 0.0735 -0.7307>}
 correction_hair_r {parent=spine_05, loc=<5.483 -0.105 3.295>, rot=<0.7289 0.0465 -0.6829 0.009>}
 correction_hair_l {parent=spine_05, loc=<5.483 -0.105 -3.295>, rot=<0.0465 -0.7289 -0.009 -0.6829>}
 thigh_l {parent=pelvis, loc=<-1.0855 -0.2109 -8.7053>, rot=<0.0009 0.0355 0.0155 -0.9993>}
 calf_l {parent=thigh_l, loc=<-42.5284 -0.0019 -0.7297>, rot=<0.0006 -0.0161 -0.0391 -0.9991>}
 foot_l {parent=calf_l, loc=<-46.7401 -2.5014 -0.592>, rot=<0.0427 -0.0238 0.0353 -0.9982>}
 thigh_r {parent=pelvis, loc=<-1.0855 -0.2109 8.7053>, rot=<0.0355 -0.0009 0.9993 0.0155>}
 calf_r {parent=thigh_r, loc=<42.5284 0.0019 0.7297>, rot=<0.0006 -0.0161 -0.0391 -0.9991>}
 foot_r {parent=calf_r, loc=<46.7401 2.5014 0.592>, rot=<0.0427 -0.0238 0.0353 -0.9982>}
]
