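# Animation Blueprint for the face accessory mesh of a female medium player cosmetic, adding dynamics on top of the copied parent and body pose.
# A Control Rig feeds two RigidBody simulations, with the backpack physics asset chosen when wearing a pack, DBNO or carrying.
# Additive offsets adjust two hair bones, then a harmonics Control Rig blends in when the movement flag is false.
# The event graph also computes gravity override and floating backpack values, which the AnimGraph never reads.
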AnimBlueprint /BRCosmetics/Characters/Player/Female/Medium/Bodies/F_MED_WordUnity_Phrase/Meshes/Parts/F_MED_WordUnity_Phrase_FaceAcc_AnimBP
target_skeleton=&/F_MED_WordUnity_Phrase_FaceAcc_Skeleton
graphs [
 AnimGraph {nodes [
   n1 = Control Rig(Source=Use cached pose 'Main + physics cache'()) {controlRigAssetReference {blueprintRigClass=&/F_MED_WordUnity_Phrase_FaceAcc_Harmonics_CtrlRig.F_MED_WordUnity_Phrase_FaceAcc_Harmonics_CtrlRig_C}}
   Output Pose(Result=Blend Poses by bool(
     BlendPose_0=Use cached pose 'Main + physics cache'()
     BlendPose_1=@n1
     BlendTime_0=1
     BlendTime_1=2
     bActiveValue=Get Movement() -> (Movement:bool)
    )) {layerGroup=None}
   n2 = Layered blend per bone(
    BasePose=Copy Pose From Mesh() {bUseAttachedParent=true}
    BlendPoses_0=Copy Pose From Mesh(SourceMeshComponent=Get Body() -> (Body:Skeletal Mesh Component Object Reference)) {bUseAttachedParent=true}
   ) {layerSetup [{branchFilters [{blendDepth=2, boneName=clavicle_r}, {blendDepth=2, boneName=clavicle_l}, {blendDepth=1, boneName=thigh_l}, {blendDepth=1, boneName=thigh_r}]}]}
   n3 = Control Rig(Source=@n2) {controlRigAssetReference {blueprintRigClass=&/F_MED_WordUnity_Phrase_FaceAcc_CtrlRig.F_MED_WordUnity_Phrase_FaceAcc_CtrlRig_C}}
   Save cached pose 'Main Cache'(Pose=@n3) {cacheName=Main Cache}
   n4 = RigidBody!(ComponentPose=Local To Component(LocalPose=Use cached pose 'Main Cache'()), Alpha, OverrideWorldGravity:Vector="0,0,-980", ExternalForce) -> (Pose) {
    bClampLinearTranslationLimitToRefPose=true
    bOverrideWorldGravity=true
    overridePhysicsAsset=&/F_MED_WordUnity_Phrase_FaceAcc_Physics
    simulationSpace=BaseBoneSpace
    baseBoneRef {boneName=spine_02}
    overrideWorldGravity {z=-980}
    simSpaceSettings {dampingAlpha=0, maxAngularAcceleration=30, maxAngularVelocity=3, maxLinearAcceleration=2000, maxLinearVelocity=700, worldAlpha=1}
   }
   n5 = Component To Local(ComponentPose=@n4)
   n6 = RigidBody(ComponentPose=Local To Component(LocalPose=Use cached pose 'Main Cache'()), OverrideWorldGravity="0,0,-980") {
    bClampLinearTranslationLimitToRefPose=true
    bOverrideWorldGravity=true
    overridePhysicsAsset=&/F_MED_WordUnity_Phrase_FaceAcc_Backpack_Physics
    simulationSpace=BaseBoneSpace
    baseBoneRef {boneName=spine_02}
    componentAppliedLinearAccClamp=<0 0 0>
    overrideWorldGravity {z=-980}
    simSpaceSettings {
     dampingAlpha=0
     maxAngularAcceleration=30
     maxAngularVelocity=3
     maxLinearAcceleration=2000
     maxLinearVelocity=700
     velocityScaleZ=0.5
     worldAlpha=1
    }
   }
   n7 = Component To Local(ComponentPose=@n6)
   n8 = Get Disable Half Dynamics() -> (Disable Half Dynamics:bool)
   n9 = "Transform (Modify) Bone - Bone: dyn_hair_03"(
    ComponentPose=Local To Component(LocalPose=Blend Poses by bool(BlendPose_0=@n7, BlendPose_1=@n5, bActiveValue=@n8) {blendType=Linear})
    Alpha=1
    Translation="0,0,-7"
    Rotation="0,0,8"
    Scale="1,1,1"
   ) {rotationMode=BMM_Additive, translationMode=BMM_Additive, boneToModify {boneName=dyn_hair_03}, rotation {roll=8}, translation {z=-7}}
   n10 = "Transform (Modify) Bone - Bone: dyn_hair_04"(ComponentPose=@n9, Alpha=1, Translation="0,0,-2", Rotation="0,0,8", Scale="1,1,1") {rotationMode=BMM_Additive, translationMode=BMM_Additive, boneToModify {boneName=dyn_hair_04}, rotation {roll=8}, translation {z=-2}}
   Save cached pose 'Main + physics cache'(Pose=Component To Local(ComponentPose=@n10)) {cacheName=Main + physics cache}
  ]}
 EventGraph {nodes [
   n1 = Event BlueprintUpdateAnimation()
   > Set Delta Time X(Delta Time X=@n1.DeltaTimeX)
   > n2 = Sequence!(execute) -> (then_0, then_1, then_2:Exec, then_3:Exec)
   n3 = Branch(execute=@n2.then_3, Condition=AND Boolean(A=Get Is Wearing Pack() -> (Is Wearing Pack:bool), B=Get bIsSkydiveFloating()))
   n4 = float / float(A=Get Delta Time X(), B=2)
   > Set FloatingBackpack(FloatingBackpack:double=FInterpTo(Current=Get FloatingBackpack() -> (FloatingBackpack:double), Target=1, DeltaTime=@n4, InterpSpeed=20)) -> (Output_Get:double)
   Set FloatingBackpack(execute=@n3.else, FloatingBackpack=FInterpTo(Current=Get FloatingBackpack(), Target=0, DeltaTime=@n4, InterpSpeed=20))
   Set Disable Half Dynamics(
    execute=@n2.then_2
    Disable Half Dynamics:bool=OR Boolean!(A=Get Is Wearing Pack(), B=Get bIsDBNO(), C:bool=Get bIsDBNOCarrying() -> (bIsDBNOCarrying:bool)) -> (ReturnValue)
   ) -> (Output_Get:bool)
   n5 = "Gravity Override State Function (Parameter Asset) (Nativized)"(
    execute=@n2.then_0
    CharacterPartAnimInstance=Self-Reference()
    GameTimeInSeconds=GetGameTimeInSeconds()
    Parameters=/BRCosmetics/Characters/Player/Female/Medium/Bodies/F_MED_WordUnity_Phrase/Meshes/F_MED_WordUnity_Phrase_Parameters
   )
   > Set Gravity Override(Gravity Override=@n5)
   n6 = PlayerMovement(execute=@n2.then_1, AnimInstance=Self-Reference(), XDotProductCoefficient=1, YDotProductCoefficient=1, ZDotProductCoefficient=1)
   n7 = "Equal (Float)"(A=@n6)
   n8 = "Equal (Float)"(A=@n6.YMovement)
   > Set Movement(Movement:bool=NOT Boolean(A=OR Boolean!(A=@n7, B=@n8) -> (ReturnValue))) -> (Output_Get:bool)
   n9 = Event OnCharacterCustomizationLoaded()
   > Set Body(Body:Skeletal Mesh Component Object Reference=@n9.CustomBaseMesh) -> (Output_Get:Skeletal Mesh Component Object Reference)
   n10 = CheckForBackbling()
   > Set Is Wearing Pack(Is Wearing Pack:bool=OR Boolean!(A=@n10, B=@n10.bIsBackblingACape, C:bool=@n10.bIsBackblingLong) -> (ReturnValue)) -> (Output_Get:bool)
  ]}
]
variables [
 Body {type=SkeletalMeshComponent}
 Delta Time X {type=number}
 Disable Half Dynamics {type=boolean}
 FloatingBackpack {type=number}
 Gravity Override {type=Vector}
 Is Wearing Pack {type=boolean}
 Movement {type=boolean}
]
